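# Physics asset for the cape on the female medium Wayfare Mask character body, pairing capsule cape bodies with simplified body shapes.
# A four-segment middle chain and a three-segment left chain descend from MSTR bodies, with tight angular limits between segments.
# Capsules on pelvis, spine, arms and thighs, plus a large low root sphere, approximate the character.
# Free-angle constraints join these shapes, tie both MSTR bodies to root, and cross-link the left chain's tip to the middle chain.
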PhysicsAsset /BRCosmetics/Characters/Player/Female/Medium/Bodies/F_MED_Wayfare_Mask/Meshes/F_MED_Wayfare_Mask_dyn_cape
bodies {
 dyn_cap_01_mid [{shapeType=Capsule, length=11.2629, radius=11, shapeName=dyn_cap_a_01_l_capsule, center=<5 0 7>, rotation=<rot 0 0 90>}]
 dyn_cap_02_mid [{shapeType=Capsule, length=14.1422, radius=11, shapeName=dyn_cap_a_02_l_capsule, center=<8 0 7>, rotation=<rot 0 0 90>}]
 dyn_cap_03_mid [{shapeType=Capsule, length=14.1422, radius=11, shapeName=dyn_cap_a_03_l_capsule, center=<8 0 7>, rotation=<rot 0 0 90>}]
 dyn_cap_04_mid [{shapeType=Capsule, length=6.4593, radius=11, shapeName=dyn_cap_a_04_l_capsule, center=<8 0 7>, rotation=<rot 0 0 90>}]
 dyn_cap_a_01_l [{shapeType=Capsule, length=7.9566, radius=9, shapeName=dyn_cap_a_01_l_capsule, center=<6.5 0 5>, rotation=<rot -0.0004 0 111.1732>}]
 dyn_cap_a_02_l [{shapeType=Capsule, length=7.9566, radius=9, shapeName=dyn_cap_a_02_l_capsule, center=<8 0 5>, rotation=<rot -0.0004 0 106.1731>}]
 dyn_cap_a_03_l [{shapeType=Capsule, length=9, radius=9, shapeName=dyn_cap_a_03_l_capsule, center=<10 0 5>, rotation=<rot -0.0004 0 106.1732>}]
 dyn_cap_a_l_MSTR [{shapeType=Capsule, length=0.505, radius=0.505, shapeName=dyn_cap_a_l_MSTR_capsule, center=<0 0 7>, rotation=<rot 0 -90 0>}]
 dyn_cap_mid_MSTR [{shapeType=Capsule, length=0.505, radius=0.505, shapeName=dyn_cap_mid_MSTR_capsule, rotation=<rot 0 -90 0>}]
 lowerarm_l [{
   shapeType=Capsule
   length=14.9406
   radius=4.1633
   shapeName=upperarm_r_capsule
   center=<11.8568 0.2347 -0.0004>
   rotation=<rot 90 0 173.1295>
  }]
 lowerarm_r [{
   shapeType=Capsule
   length=14.9406
   radius=4.1633
   shapeName=upperarm_r_capsule
   center=<-11.8568 -0.2347 0.0004>
   rotation=<rot 90 -90 -91.6498>
  }]
 pelvis [{shapeType=Capsule, length=11.1444, radius=14.5035, center=<-3.9432 1.8194 2.7941>}]
 root [{shapeType=Sphere, radius=80.5, shapeName=root_sphere, center=<0 6.6224 -58.6046>}]
 spine_02 [{shapeType=Capsule, length=2.1924, radius=12.3197, center=<-0.8286 4.0772 0.0003>, rotation=<rot 90 0 0>}]
 thigh_l [{shapeType=Capsule, length=19.7839, radius=10.2396, shapeName=thigh_l_capsule, center=<-18.3966 1.1085 0>, rotation=<rot 90 89.3553 94.7975>}]
 thigh_r [{shapeType=Capsule, length=19.7839, radius=10.2396, shapeName=thigh_l_capsule, center=<18.3966 -1.1085 0>, rotation=<rot 90 164.707 -9.8508>}]
 upperarm_l [{
   shapeType=Capsule
   length=14.9406
   radius=4.1633
   shapeName=upperarm_r_capsule
   center=<23.5274 -0.2995 0.0001>
   rotation=<rot 90 -90 94.2852>
  }]
 upperarm_r [{shapeType=Capsule, length=14.9406, radius=4.1633, shapeName=upperarm_r_capsule, center=<-23.5274 0.2995 -0.0001>, rotation=<rot 90 0 4.2852>}]
}
constraints [
 {
  bone1Name=spine_02
  bone2Name=pelvis
  swing1LimitDegrees=45
  swing1Motion=Free
  swing2LimitDegrees=45
  swing2Motion=Free
  twistLimitDegrees=45
  twistMotion=Free
 }
 {bone1Name=dyn_cap_a_01_l, bone2Name=dyn_cap_a_l_MSTR, swing1LimitDegrees=10, swing2LimitDegrees=30, twistLimitDegrees=2}
 {bone1Name=dyn_cap_a_02_l, bone2Name=dyn_cap_a_01_l, swing1LimitDegrees=20, swing2LimitDegrees=30, twistLimitDegrees=2}
 {bone1Name=dyn_cap_a_03_l, bone2Name=dyn_cap_a_02_l, swing1LimitDegrees=20, swing2LimitDegrees=30, twistLimitDegrees=2}
 {bone1Name=dyn_cap_01_mid, bone2Name=dyn_cap_mid_MSTR, swing1LimitDegrees=10, swing2LimitDegrees=35, twistLimitDegrees=10}
 {bone1Name=dyn_cap_02_mid, bone2Name=dyn_cap_01_mid, swing1LimitDegrees=20, swing2LimitDegrees=35, twistLimitDegrees=10}
 {bone1Name=dyn_cap_03_mid, bone2Name=dyn_cap_02_mid, swing1LimitDegrees=20, swing2LimitDegrees=35, twistLimitDegrees=10}
 {bone1Name=dyn_cap_04_mid, bone2Name=dyn_cap_03_mid, swing1LimitDegrees=20, swing2LimitDegrees=35, twistLimitDegrees=10}
 {
  bone1Name=upperarm_r
  bone2Name=spine_02
  swing1LimitDegrees=45
  swing1Motion=Free
  swing2LimitDegrees=45
  swing2Motion=Free
  twistLimitDegrees=45
  twistMotion=Free
 }
 {
  bone1Name=upperarm_l
  bone2Name=spine_02
  swing1LimitDegrees=45
  swing1Motion=Free
  swing2LimitDegrees=45
  swing2Motion=Free
  twistLimitDegrees=45
  twistMotion=Free
 }
 {
  bone1Name=lowerarm_l
  bone2Name=upperarm_l
  swing1LimitDegrees=45
  swing1Motion=Free
  swing2LimitDegrees=45
  swing2Motion=Free
  twistLimitDegrees=45
  twistMotion=Free
 }
 {
  bone1Name=lowerarm_r
  bone2Name=upperarm_r
  swing1LimitDegrees=45
  swing1Motion=Free
  swing2LimitDegrees=45
  swing2Motion=Free
  twistLimitDegrees=45
  twistMotion=Free
 }
 {
  bone1Name=dyn_cap_a_03_l
  bone2Name=dyn_cap_03_mid
  swing1LimitDegrees=45
  swing1Motion=Free
  swing2LimitDegrees=45
  swing2Motion=Free
  twistLimitDegrees=45
  twistMotion=Free
 }
 {
  bone1Name=thigh_l
  bone2Name=pelvis
  swing1LimitDegrees=45
  swing1Motion=Free
  swing2LimitDegrees=45
  swing2Motion=Free
  twistLimitDegrees=45
  twistMotion=Free
 }
 {
  bone1Name=thigh_r
  bone2Name=pelvis
  swing1LimitDegrees=45
  swing1Motion=Free
  swing2LimitDegrees=45
  swing2Motion=Free
  twistLimitDegrees=45
  twistMotion=Free
 }
 {
  bone1Name=pelvis
  bone2Name=root
  swing1LimitDegrees=45
  swing1Motion=Free
  swing2LimitDegrees=45
  swing2Motion=Free
  twistLimitDegrees=45
  twistMotion=Free
 }
 {
  bone1Name=dyn_cap_mid_MSTR
  bone2Name=root
  swing1LimitDegrees=45
  swing1Motion=Free
  swing2LimitDegrees=45
  swing2Motion=Free
  twistLimitDegrees=45
  twistMotion=Free
 }
 {
  bone1Name=dyn_cap_a_l_MSTR
  bone2Name=root
  swing1LimitDegrees=45
  swing1Motion=Free
  swing2LimitDegrees=45
  swing2Motion=Free
  twistLimitDegrees=45
  twistMotion=Free
 }
]
