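# Physics asset for the left cape on the Wayfare Mask female medium player body cosmetic, apparently for secondary cape motion.
# Capsule bodies form a constraint chain from spine_04 through the left upper-arm twist bone and a cape master bone to the cape bone.
# Only the cape-to-master joint limits rotation, with very tight twist; the other joints rotate freely.
# The spine and cape have sizable capsules, while the intermediate bones carry tiny ones.
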PhysicsAsset /BRCosmetics/Characters/Player/Female/Medium/Bodies/F_MED_Wayfare_Mask/Meshes/F_MED_Wayfare_Mask_dyn_cape_left
bodies {
 def_cape_l [{shapeType=Capsule, length=8, radius=9, shapeName=dyn_cap_a_01_l_capsule, center=<10 0 0>, rotation=<rot 90 0 0>}]
 def_cape_l_MSTR [{shapeType=Capsule, length=1, radius=1, shapeName=def_cape_l_MSTR_capsule}]
 spine_04 [{
   shapeType=Capsule
   length=4.1066
   radius=11.8842
   shapeName=spine_04_capsule
   center=<-0.7946 1.056 -0.3885>
   rotation=<rot 0.8438 -90 0.9299>
  }]
 upperarm_twist_01_l [{shapeType=Capsule, length=1, radius=1, shapeName=upperarm_twist_01_l_capsule}]
}
constraints [
 {
  bone1Name=def_cape_l_MSTR
  bone2Name=upperarm_twist_01_l
  swing1LimitDegrees=45
  swing1Motion=Free
  swing2LimitDegrees=45
  swing2Motion=Free
  twistLimitDegrees=45
  twistMotion=Free
 }
 {bone1Name=def_cape_l, bone2Name=def_cape_l_MSTR, swing1LimitDegrees=25, swing2LimitDegrees=45, twistLimitDegrees=2}
 {
  bone1Name=upperarm_twist_01_l
  bone2Name=spine_04
  swing1LimitDegrees=45
  swing1Motion=Free
  swing2LimitDegrees=45
  swing2Motion=Free
  twistLimitDegrees=45
  twistMotion=Free
 }
]
