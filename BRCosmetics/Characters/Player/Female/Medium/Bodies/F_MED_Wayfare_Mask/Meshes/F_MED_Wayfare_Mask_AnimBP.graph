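# Post-process style AnimBP for the Wayfare Mask female body cosmetic, adding secondary motion to cape, armor and zipper.
# A cached pre-pose copies the parent mesh, runs a deform Control Rig, applies correctives from limb-rotation alphas and simulates right-armor AnimDynamics.
# RigidBody cape simulation switches physics assets for back bling, with a bone-constraint fallback, before zipper dynamics.
# Updates compute gravity/wind override and limb alphas, and detect the front-end idle montage to adjust the cape.
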AnimBlueprint /BRCosmetics/Characters/Player/Female/Medium/Bodies/F_MED_Wayfare_Mask/Meshes/F_MED_Wayfare_Mask_AnimBP
target_skeleton=&/F_MED_Wayfare_Mask_Skeleton
graphs [
 AnimGraph {nodes [
   Look At - Bone: dyn_cap_a_01_l(Alpha=1) {
    alphaBoolBlend {blendOption=Linear}
    boneToModify {boneName=dyn_cap_a_01_l}
    lookAtLocation=<-20 -15 -20>
    lookAtTarget {boneReference {boneName=spine_01}}
    lookAt_Axis {axis=<1 0>}
   }
   n1 = Get GravityOverride()
   n2 = RigidBody!(ComponentPose=Local To Component(LocalPose=Use cached pose 'PrePose'()), Alpha, OverrideWorldGravity:Vector=@n1, ExternalForce) -> (Pose) {
    bClampLinearTranslationLimitToRefPose=true
    bOverrideWorldGravity=true
    overridePhysicsAsset=&/F_MED_Wayfare_Mask_dyn_cape
    simulationSpace=BaseBoneSpace
    alphaBoolBlend {blendOption=Linear}
    baseBoneRef {boneName=spine_01}
    componentAppliedLinearAccClamp=<3000 3000 3000>
    componentLinearAccScale {x=0.5}
    componentLinearVelScale {x=0.5}
    overrideWorldGravity {z=-980}
    simSpaceSettings {
     maxAngularAcceleration=100
     maxAngularVelocity=5
     maxLinearAcceleration=500
     maxLinearVelocity=50
     velocityScaleZ=0.5
     worldAlpha=1
     externalLinearDragV=<1 1 1>
    }
   }
   n3 = Component To Local(ComponentPose=@n2)
   n4 = Look At - Bone: dyn_cap_01_mid(ComponentPose=Local To Component(LocalPose=Use cached pose 'PrePose'()), Alpha=1) {
    alphaBoolBlend {blendOption=Linear}
    boneToModify {boneName=dyn_cap_01_mid}
    lookAtLocation=<-20 -20>
    lookAtTarget {boneReference {boneName=spine_01}}
    lookAt_Axis {axis=<1 0>}
   }
   n5 = RigidBody(ComponentPose=@n4, OverrideWorldGravity=@n1) {
    bClampLinearTranslationLimitToRefPose=true
    bOverrideWorldGravity=true
    overridePhysicsAsset=&/F_MED_Wayfare_Mask_dyn_cape_wearing_pack
    simulationSpace=BaseBoneSpace
    alphaBoolBlend {blendOption=Linear}
    baseBoneRef {boneName=spine_01}
    componentAppliedLinearAccClamp=<3000 3000 3000>
    componentLinearAccScale {x=0.5}
    componentLinearVelScale {x=0.5}
    overrideWorldGravity {z=-980}
    simSpaceSettings {
     maxAngularAcceleration=100
     maxAngularVelocity=5
     maxLinearAcceleration=500
     maxLinearVelocity=50
     velocityScaleZ=0.5
     worldAlpha=1
     externalLinearDragV=<1 1 1>
    }
   }
   n6 = Component To Local(ComponentPose=@n5)
   n7 = Get isWearBackbling() -> (isWearBackbling:bool)
   n8 = RigidBody(
    ComponentPose=Local To Component(LocalPose=Blend Poses by bool(BlendPose_0=@n6, BlendPose_1=@n3, BlendTime_0=0.3, BlendTime_1=0.3, bActiveValue=@n7) {blendType=Linear})
    OverrideWorldGravity="0,0,-980"
   ) {
    bClampLinearTranslationLimitToRefPose=true
    bOverrideWorldGravity=true
    overridePhysicsAsset=&/F_MED_Wayfare_Mask_dyn_cape_left
    simulationSpace=BaseBoneSpace
    alphaBoolBlend {blendOption=Linear}
    baseBoneRef {boneName=upperarm_twist_01_l}
    componentAppliedLinearAccClamp=<500 500 500>
    overrideWorldGravity {z=-980}
    simSpaceSettings {
     maxAngularAcceleration=100
     maxAngularVelocity=5
     maxLinearAcceleration=500
     maxLinearVelocity=50
     velocityScaleZ=0.5
     externalLinearDragV=<1 1 1>
    }
   }
   n9 = Component To Local(ComponentPose=@n8)
   n10 = "Transform (Modify) Bone - Bone: def_cape_l_MSTR"(
    ComponentPose=Local To Component(LocalPose=Use cached pose 'PrePose'())
    Alpha=Get arm_rot_dn_l() -> (arm_rot_dn_l:double)
    Translation="0,0,0"
    Rotation="30,0,0"
    Scale="1,1,1"
   ) {
    rotationMode=BMM_Additive
    rotationSpace=BCS_BoneSpace
    translationMode=BMM_Additive
    translationSpace=BCS_BoneSpace
    alphaBoolBlend {blendOption=Linear}
    boneToModify {boneName=def_cape_l_MSTR}
    rotation {pitch=30}
   }
   n11 = "Transform (Modify) Bone - Bone: def_cape_l_MSTR"(ComponentPose=@n10, Alpha=Get arm_twist_l() -> (arm_twist_l:double), Translation="0,0,0", Rotation="0,-50,0.000", Scale="1,1,1") {
    rotationMode=BMM_Additive
    rotationSpace=BCS_BoneSpace
    translationMode=BMM_Additive
    translationSpace=BCS_BoneSpace
    alphaBoolBlend {blendOption=Linear}
    boneToModify {boneName=def_cape_l_MSTR}
    rotation {yaw=-50}
   }
   n12 = Get capeConstraintAplha() -> (capeConstraintAplha:double)
   n13 = "Constraint - dyn_cap_02_mid (PPPPP)"!(
    ComponentPose=@n11
    Alpha=@n12
    ConstraintWeights_0=0.4878
    ConstraintWeights_1:float=0.3209
    ConstraintWeights_2:float=0.0928
    ConstraintWeights_3:float=0.0744
    ConstraintWeights_4:float=0.0242
   ) -> (Pose) {
    alphaBoolBlend {blendOption=Linear}
    boneToModify {boneName=dyn_cap_02_mid}
    constraintSetup [
     {targetBone {boneName=pelvis}}
     {targetBone {boneName=spine_01}}
     {targetBone {boneName=spine_02}}
     {targetBone {boneName=spine_03}}
     {targetBone {boneName=spine_04}}
    ]
    constraintWeights [1,1,1,1,1]
   }
   n14 = "Constraint - dyn_cap_a_02_l (PPPPP)"!(
    ComponentPose=@n13
    Alpha=@n12
    ConstraintWeights_0=0.4878
    ConstraintWeights_1:float=0.3209
    ConstraintWeights_2:float=0.0928
    ConstraintWeights_3:float=0.0744
    ConstraintWeights_4:float=0.0242
   ) -> (Pose) {
    alphaBoolBlend {blendOption=Linear}
    boneToModify {boneName=dyn_cap_a_02_l}
    constraintSetup [
     {targetBone {boneName=pelvis}}
     {targetBone {boneName=spine_01}}
     {targetBone {boneName=spine_02}}
     {targetBone {boneName=spine_03}}
     {targetBone {boneName=spine_04}}
    ]
    constraintWeights [1,1,1,1,1]
   }
   n15 = "Constraint - dyn_cap_a_03_l (P)"(ComponentPose=@n14, Alpha=@n12, ConstraintWeights_0=1) {
    alphaBoolBlend {blendOption=Linear}
    boneToModify {boneName=dyn_cap_a_03_l}
    constraintSetup [{targetBone {boneName=pelvis}}]
    constraintWeights [1]
   }
   n16 = "Constraint - dyn_cap_03_mid (P)"(ComponentPose=@n15, Alpha=@n12, ConstraintWeights_0=1) {
    alphaBoolBlend {blendOption=Linear}
    boneToModify {boneName=dyn_cap_03_mid}
    constraintSetup [{targetBone {boneName=pelvis}}]
    constraintWeights [1]
   }
   n17 = Component To Local(ComponentPose=@n16)
   n18 = Get bRigidBodyEnabled()
   n19 = "Transform (Modify) Bone - Bone: dyn_cap_a_l_MSTR"(
    ComponentPose=Local To Component(LocalPose=Blend Poses by bool(BlendPose_0=@n9, BlendPose_1=@n17, BlendTime_0=0.3, BlendTime_1=0.3, bActiveValue=@n18) {blendType=Linear})
    Alpha=1
    Translation="0,-9,4"
    Rotation="10,-10,0"
    Scale="1,1,1"
   ) {
    rotationMode=BMM_Additive
    rotationSpace=BCS_BoneSpace
    translationMode=BMM_Additive
    translationSpace=BCS_BoneSpace
    alphaBoolBlend {blendInTime=0.3, blendOption=Linear}
    alphaScaleBiasClamp {bInterpResult=true, interpSpeedIncreasing=15}
    boneToModify {boneName=dyn_cap_a_l_MSTR}
    rotation {pitch=10, yaw=-10}
    translation {y=-9, z=4}
   }
   n20 = "Transform (Modify) Bone - Bone: dyn_cap_02_mid"(ComponentPose=@n19, Alpha=1, Translation="0,-5,0", Rotation="15,-5,0", Scale="1,1,1") {
    rotationMode=BMM_Additive
    rotationSpace=BCS_BoneSpace
    translationMode=BMM_Additive
    translationSpace=BCS_BoneSpace
    alphaBoolBlend {blendInTime=0.3, blendOption=Linear}
    alphaScaleBiasClamp {bInterpResult=true, interpSpeedIncreasing=15}
    boneToModify {boneName=dyn_cap_02_mid}
    rotation {pitch=15, yaw=-5}
    translation {y=-5}
   }
   n21 = "Transform (Modify) Bone - Bone: def_cape_l_MSTR"(ComponentPose=@n20, Alpha=1, Translation="0,-5,0", Rotation="0,-15,0", Scale="1,1,1") {
    rotationMode=BMM_Additive
    rotationSpace=BCS_BoneSpace
    translationMode=BMM_Additive
    translationSpace=BCS_BoneSpace
    alphaBoolBlend {blendInTime=0.3, blendOption=Linear}
    alphaScaleBiasClamp {bInterpResult=true, interpSpeedIncreasing=15}
    boneToModify {boneName=def_cape_l_MSTR}
    rotation {yaw=-15}
    translation {y=-5}
   }
   n22 = "Transform (Modify) Bone - Bone: dyn_cap_a_l_MSTR"!(ComponentPose=@n21, bAlphaBoolEnabled:bool=Get bIsSprinting(), Translation="0,0,0", Rotation="20,0,0", Scale="1,1,1") -> (Pose) {
    alphaInputType=Bool
    rotationMode=BMM_Additive
    rotationSpace=BCS_BoneSpace
    translationMode=BMM_Additive
    translationSpace=BCS_BoneSpace
    alphaBoolBlend {blendInTime=0.2, blendOption=Linear, blendOutTime=0.5}
    boneToModify {boneName=dyn_cap_a_l_MSTR}
    rotation {pitch=20}
   }
   n23 = "Constraint - dyn_zipper_MSTR (PPP)"!(ComponentPose=@n22, Alpha=1, ConstraintWeights_0=0.1157, ConstraintWeights_1:float=0.388, ConstraintWeights_2:float=0.4963) -> (Pose) {
    alphaBoolBlend {blendOption=Linear}
    boneToModify {boneName=dyn_zipper_MSTR}
    constraintSetup [{targetBone {boneName=spine_04}}, {targetBone {boneName=spine_05}}, {targetBone {boneName=def_cape_l}}]
    constraintWeights [1,1,1]
   }
   n24 = "Constraint - dyn_zipper_TGT (PPPPPP)"!(
    ComponentPose=@n23
    Alpha=1
    ConstraintWeights_0=0.1335
    ConstraintWeights_1:float=0.3244
    ConstraintWeights_2:float=0.1406
    ConstraintWeights_3:float=0.0175
    ConstraintWeights_4:float=0.3583
    ConstraintWeights_5:float=0.0257
   ) -> (Pose) {
    alphaBoolBlend {blendOption=Linear}
    boneToModify {boneName=dyn_zipper_TGT}
    constraintSetup [
     {targetBone {boneName=spine_05}}
     {targetBone {boneName=dyn_cap_a_01_l}}
     {targetBone {boneName=dyn_cap_01_mid}}
     {targetBone {boneName=dyn_cap_02_mid}}
     {targetBone {boneName=def_cape_l}}
     {targetBone {boneName=dyn_cap_a_02_l}}
    ]
    constraintWeights [1,1,1,1,1,1]
   }
   n25 = Look At - Bone: dyn_zipper_MSTR(ComponentPose=@n24, Alpha=1) {
    alphaBoolBlend {blendOption=Linear}
    boneToModify {boneName=dyn_zipper_MSTR}
    lookAtLocation {x=0, z=0.5}
    lookAtTarget {boneReference {boneName=dyn_zipper_TGT}}
    lookAt_Axis {axis=<1 0>}
   }
   n26 = Anim Dynamics - Bone: dyn_zipper(ComponentPose=@n25, Alpha=1) {
    angularDampingOverride=0.999
    angularSpringConstant=500
    bAngularSpring=true
    bOverrideAngularDamping=true
    bUseGravityOverride=true
    simulationSpace=RootRelative
    alphaBoolBlend {blendOption=Linear}
    boundBone {boneName=dyn_zipper}
    componentAppliedLinearAccClamp=<3000 3000 3000>
    componentLinearAccScale=<0.2 0.3 0.3>
    componentLinearVelScale=<0.2 0.3 0.3>
    gravityOverride {z=-980}
    physicsBodyDefinitions [{
      boundBone {boneName=dyn_zipper}
      boxExtents=<15 10 10>
      constraintSetup {angularLimitsMax=<0 0 15>, angularLimitsMin=<0 -60 -15>, angularTarget=<0 0 0>}
      localJointOffset=<5 0 0>
     }]
   }
   Output Pose(Result=Component To Local(ComponentPose=@n26)) {layerGroup=None}
   n27 = Control Rig(Source=Copy Pose From Mesh() {bUseAttachedParent=true}) {
    alphaBoolBlend {blendOption=Linear}
    controlRigAssetReference {blueprintRigClass=&/F_MED_Wayfare_Mask_DeformRig_ControlRig.F_MED_Wayfare_Mask_DeformRig_ControlRig_C}
   }
   n28 = "Transform (Modify) Bone - Bone: dyn_knee_armor_l"(
    ComponentPose=Copy Bone - Source Bone: head - Target Bone: head_offset!(
     ComponentPose:Component Space Pose Link Structure=Local To Component(LocalPose=@n27)
     Alpha:float=0.2
     bCopyTranslation:bool=true
     bCopyRotation:bool=true
     bCopyScale:bool=true
    ) -> (Pose:Component Space Pose Link Structure) {is=title AnimGraphNode_CopyBone, alphaBoolBlend {blendOption=Linear}, sourceBone {boneName=head}, targetBone {boneName=head_offset}}
    Alpha=Get knee_rot_l() -> (knee_rot_l:double)
    Translation="0,0,0.2"
    Rotation="-40,0,0"
    Scale="1.000,1.05,1.000"
   ) {
    rotationMode=BMM_Additive
    rotationSpace=BCS_BoneSpace
    scaleMode=BMM_Additive
    scaleSpace=BCS_BoneSpace
    translationMode=BMM_Additive
    translationSpace=BCS_BoneSpace
    alphaBoolBlend {blendOption=Linear}
    boneToModify {boneName=dyn_knee_armor_l}
    rotation {pitch=-40}
    scale {y=1.05}
    translation {z=0.2}
   }
   n29 = "Transform (Modify) Bone - Bone: dyn_knee_armor_r"(ComponentPose=@n28, Alpha=Get knee_rot_r() -> (knee_rot_r:double), Translation="0,0,0.2", Rotation="-40,0,0", Scale="1.000,1.05,1.000") {
    rotationMode=BMM_Additive
    rotationSpace=BCS_BoneSpace
    scaleMode=BMM_Additive
    scaleSpace=BCS_BoneSpace
    translationMode=BMM_Additive
    translationSpace=BCS_BoneSpace
    alphaBoolBlend {blendOption=Linear}
    boneToModify {boneName=dyn_knee_armor_r}
    rotation {pitch=-40}
    scale {y=1.05}
    translation {z=0.2}
   }
   n30 = Get arm_rot_bk_l() -> (arm_rot_bk_l:double)
   n31 = "Transform (Modify) Bone - Bone: dyn_cap_a_l_MSTR"!(ComponentPose=@n29, Alpha:float=@n30, Translation="0,-3,10", Rotation="0,0,0", Scale="1,1,1") -> (Pose) {
    rotationMode=BMM_Additive
    rotationSpace=BCS_BoneSpace
    translationMode=BMM_Additive
    translationSpace=BCS_BoneSpace
    alphaBoolBlend {blendOption=Linear}
    boneToModify {boneName=dyn_cap_a_l_MSTR}
    translation {y=-3, z=10}
   }
   n32 = "Transform (Modify) Bone - Bone: def_cape_l_MSTR"(ComponentPose=@n31, Alpha=@n30, Translation="0,-8,0", Rotation="0,0,20", Scale="1,1,1") {
    rotationMode=BMM_Additive
    rotationSpace=BCS_BoneSpace
    translationMode=BMM_Additive
    translationSpace=BCS_BoneSpace
    alphaBoolBlend {blendOption=Linear}
    boneToModify {boneName=def_cape_l_MSTR}
    rotation {roll=20}
    translation {y=-8}
   }
   n33 = Get arm_rot_up_l() -> (arm_rot_up_l:double)
   n34 = "Transform (Modify) Bone - Bone: def_cape_l_MSTR"(ComponentPose=@n32, Alpha=@n33, Translation="0,0,0", Rotation="-30,0,0", Scale="1,1,1") {
    rotationMode=BMM_Additive
    rotationSpace=BCS_BoneSpace
    translationMode=BMM_Additive
    translationSpace=BCS_BoneSpace
    alphaBoolBlend {blendOption=Linear}
    boneToModify {boneName=def_cape_l_MSTR}
    rotation {pitch=-30}
   }
   n35 = "Transform (Modify) Bone - Bone: dyn_cap_a_l_MSTR"(ComponentPose=@n34, Alpha=@n33, Translation="-6,2,-3", Rotation="0,0,0", Scale="1,1,1") {
    rotationMode=BMM_Additive
    rotationSpace=BCS_BoneSpace
    translationMode=BMM_Additive
    translationSpace=BCS_BoneSpace
    alphaBoolBlend {blendOption=Linear}
    boneToModify {boneName=dyn_cap_a_l_MSTR}
    translation=<-6 2 -3>
   }
   n36 = "Transform (Modify) Bone - Bone: def_hood_offset"(ComponentPose=@n35, Alpha=Get arm_rot_up_r() -> (arm_rot_up_r:double), Translation="10,0,0", Rotation="0,0,0", Scale="1,1,1") {
    rotationMode=BMM_Additive
    rotationSpace=BCS_BoneSpace
    scaleMode=BMM_Replace
    scaleSpace=BCS_BoneSpace
    translationMode=BMM_Additive
    translationSpace=BCS_BoneSpace
    alphaBoolBlend {blendOption=Linear}
    alphaScaleBiasClamp {bMapRange=true, inRange {max=0.5}}
    boneToModify {boneName=def_hood_offset}
    translation {x=10}
   }
   n37 = Get isIdleEmotePlaying() -> (isIdleEmotePlaying:bool)
   n38 = "Transform (Modify) Bone - Bone: dyn_cap_a_l_MSTR"!(ComponentPose=@n36, bAlphaBoolEnabled:bool=@n37, Translation="0,-3,3", Rotation="15,0,0", Scale="1,1,1") -> (Pose) {
    alphaInputType=Bool
    rotationMode=BMM_Additive
    rotationSpace=BCS_BoneSpace
    translationMode=BMM_Additive
    translationSpace=BCS_BoneSpace
    alphaBoolBlend {blendOption=Linear}
    boneToModify {boneName=dyn_cap_a_l_MSTR}
    rotation {pitch=15}
    translation {y=-3, z=3}
   }
   n39 = "Transform (Modify) Bone - Bone: def_cape_l_MSTR"!(ComponentPose=@n38, bAlphaBoolEnabled:bool=@n37, Translation="0,-1,0", Rotation="0,-15,0", Scale="1,1,1") -> (Pose) {
    alphaInputType=Bool
    rotationMode=BMM_Additive
    rotationSpace=BCS_BoneSpace
    translationMode=BMM_Additive
    translationSpace=BCS_BoneSpace
    alphaBoolBlend {blendOption=Linear}
    boneToModify {boneName=def_cape_l_MSTR}
    rotation {yaw=-15}
    translation {y=-1}
   }
   n40 = "Transform (Modify) Bone - Bone: dyn_cap_mid_MSTR"!(ComponentPose=@n39, bAlphaBoolEnabled:bool=@n37, Translation="0,0,1", Rotation="10,0,0", Scale="1,1,1") -> (Pose) {
    alphaInputType=Bool
    rotationMode=BMM_Additive
    rotationSpace=BCS_BoneSpace
    translationMode=BMM_Additive
    translationSpace=BCS_BoneSpace
    alphaBoolBlend {blendOption=Linear}
    boneToModify {boneName=dyn_cap_mid_MSTR}
    rotation {pitch=10}
    translation {z=1}
   }
   n41 = "Constraint - dyn_armor_blet_r_MSTR (PPPPP)"!(
    ComponentPose=@n40
    Alpha=1
    ConstraintWeights_0=0.0094
    ConstraintWeights_1:float=0.2517
    ConstraintWeights_2:float=0.1572
    ConstraintWeights_3:float=0.1355
    ConstraintWeights_4:float=0.4462
   ) -> (Pose) {
    alphaBoolBlend {blendOption=Linear}
    boneToModify {boneName=dyn_armor_blet_r_MSTR}
    constraintSetup [
     {targetBone {boneName=spine_04}}
     {targetBone {boneName=spine_05}}
     {targetBone {boneName=clavicle_r}}
     {targetBone {boneName=upperarm_twist_01_r}}
     {targetBone {boneName=def_shoulder_armor_r}}
    ]
    constraintWeights [1,1,1,1,1]
   }
   n42 = Get arm_rot_dn_r() -> (arm_rot_dn_r:double)
   n43 = "Transform (Modify) Bone - Bone: def_shoulder_armor_r"(ComponentPose=@n41, Alpha=@n42, Translation="0,1,0", Rotation="30,50,0", Scale="1,1,1") {
    rotationMode=BMM_Additive
    rotationSpace=BCS_BoneSpace
    scaleMode=BMM_Replace
    scaleSpace=BCS_BoneSpace
    translationMode=BMM_Additive
    translationSpace=BCS_BoneSpace
    alphaBoolBlend {blendOption=Linear}
    boneToModify {boneName=def_shoulder_armor_r}
    rotation {pitch=30, yaw=50}
    translation {y=1}
   }
   n44 = "Transform (Modify) Bone - Bone: dyn_armor_blet_r_MSTR"(ComponentPose=@n43, Alpha=@n42, Translation="0,3,-0.3", Rotation="-0.010,15.002,-5.000", Scale="1,1,1") {
    rotationMode=BMM_Additive
    rotationSpace=BCS_BoneSpace
    scaleMode=BMM_Replace
    scaleSpace=BCS_BoneSpace
    translationMode=BMM_Additive
    translationSpace=BCS_BoneSpace
    alphaBoolBlend {blendOption=Linear}
    boneToModify {boneName=dyn_armor_blet_r_MSTR}
    rotation=<rot -0.01 15.002 -5>
    translation {y=3, z=-0.3}
   }
   n45 = "Transform (Modify) Bone - Bone: dyn_armor_blet_r_MSTR"(ComponentPose=@n44, Alpha=Get arm_rot_up_r(), Translation="0,0,0", Rotation="0,-50,0", Scale="1,1,1") {
    rotationMode=BMM_Additive
    rotationSpace=BCS_BoneSpace
    scaleMode=BMM_Replace
    scaleSpace=BCS_BoneSpace
    translationMode=BMM_Additive
    translationSpace=BCS_BoneSpace
    alphaBoolBlend {blendOption=Linear}
    boneToModify {boneName=dyn_armor_blet_r_MSTR}
    rotation {yaw=-50}
   }
   n46 = "Anim Dynamics - Chain: dyn_armor_blet_r_01 -> dyn_armor_blet_r_02"(ComponentPose=@n45, Alpha=1, ExternalForce="0,0,0") {
    angularDampingOverride=0.9999
    angularSpringConstant=500
    bAngularSpring=true
    bChain=true
    bOverrideAngularDamping=true
    bUseGravityOverride=true
    simulationSpace=RootRelative
    alphaBoolBlend {blendOption=Linear}
    boundBone {boneName=dyn_armor_blet_r_01}
    chainEnd {boneName=dyn_armor_blet_r_02}
    componentAppliedLinearAccClamp=<3000 3000 3000>
    componentLinearAccScale=<0.2 0.3 0.3>
    componentLinearVelScale=<0.2 0.3 0.3>
    gravityOverride {z=-980}
    physicsBodyDefinitions [
     {
      boundBone {boneName=dyn_armor_blet_r_01}
      boxExtents=<15 10 10>
      constraintSetup {angularLimitsMax=<0 10 15>, angularLimitsMin=<0 -25 0>, angularTarget=<0 0 0>}
      localJointOffset=<2 0 0>
     }
     {
      boundBone {boneName=dyn_armor_blet_r_02}
      boxExtents=<15 10 10>
      constraintSetup {angularLimitsMax=<0 20 5>, angularLimitsMin=<0 -25 -5>, angularTarget=<0 0 0>}
      localJointOffset=<1 0 0>
     }
    ]
   }
   n47 = "Transform (Modify) Bone - Bone: dyn_armor_blet_r_MSTR"!(ComponentPose=@n46, Alpha=1, Rotation="0,4,0") -> (Pose) {
    rotationMode=BMM_Additive
    rotationSpace=BCS_BoneSpace
    scaleMode=BMM_Replace
    scaleSpace=BCS_BoneSpace
    translationMode=BMM_Additive
    translationSpace=BCS_BoneSpace
    alphaBoolBlend {blendOption=Linear}
    boneToModify {boneName=dyn_armor_blet_r_MSTR}
    rotation {yaw=4}
   }
   n48 = "Transform (Modify) Bone - Bone: def_belt_r"(
    ComponentPose=@n47
    Alpha=Get thigh_rot_up_r() -> (thigh_rot_up_r:double)
    Translation="-12,0,-1"
    Rotation="20,0,0"
    Scale="0.7,1.000,1.000"
   ) {
    rotationMode=BMM_Additive
    rotationSpace=BCS_BoneSpace
    scaleMode=BMM_Additive
    scaleSpace=BCS_BoneSpace
    translationMode=BMM_Additive
    translationSpace=BCS_BoneSpace
    alphaBoolBlend {blendOption=Linear}
    boneToModify {boneName=def_belt_r}
    rotation {pitch=20}
    scale {x=0.7}
    translation {x=-12, z=-1}
   }
   n49 = "Transform (Modify) Bone - Bone: def_belt_l"(
    ComponentPose=@n48
    Alpha=Get thigh_rot_up_l() -> (thigh_rot_up_l:double)
    Translation="-10,0,0"
    Rotation="30,-12,-10"
    Scale="0.65,1.000,1.000"
   ) {
    rotationMode=BMM_Additive
    rotationSpace=BCS_BoneSpace
    scaleMode=BMM_Replace
    scaleSpace=BCS_BoneSpace
    translationMode=BMM_Additive
    translationSpace=BCS_BoneSpace
    alphaBoolBlend {blendOption=Linear}
    boneToModify {boneName=def_belt_l}
    rotation=<rot 30 -12 -10>
    scale {x=0.65}
    translation {x=-10}
   }
   Save cached pose 'PrePose'(Pose=Component To Local(ComponentPose=@n49)) {cacheName=PrePose}
   Get bIsTacticalSprinting()
  ]}
 EventGraph {nodes [
   Event OnCharacterCustomizationLoaded()
   > Do Once()
   > Set isWearBackbling(isWearBackbling:bool=CheckForBackbling()) -> (Output_Get:bool)
   Event BlueprintInitializeAnimation()
   > Set SkelMesh(SkelMesh=GetOwnerSkeletalMeshComponent())
   n1 = TryGetFortPawn()
   > Is Valid(InputObject=@n1)
   > n2 = Cast To FortPlayerPawnAthena(Object=@n1)
   > Is Valid(InputObject=@n2)
   > n3 = Set Player_Pawn(Player_Pawn:Fort Player Pawn Object Reference=@n2) -> (Output_Get:Fort Player Pawn Object Reference)
   n4 = GetAnimInstance(self=Get Mesh(self=@n3))
   > Is Valid(InputObject=@n4)
   > Set Anim_Inst(Anim_Inst:Anim Instance Object Reference=@n4) -> (Output_Get:Anim Instance Object Reference)
   Event BlueprintUpdateAnimation()
   > n5 = Sequence!(execute) -> (then_0, then_1, then_2:Exec)
   > n6 = Gravity Override State Function DEPRECATED(CharacterPartAnimInstance=Self-Reference(), GameTimeInSeconds=GetGameTimeInSeconds(), GravityOverrideParams=Get Gravity Override Params())
   > Set GravityOverride(GravityOverride=@n6)
   n7 = Get SkelMesh()
   Set knee_rot_l(
    execute=@n5.then_1
    knee_rot_l:double=GetAlphaFromReferenceTransformDelta(SkeletalMeshComponent=@n7, BoneName=calf_l, BaseName=thigh_l, DeltaComponentType=QuaternionTwist, InputRange=(X=0,Y=-110), TwistAxis="0,0,1")
   ) -> (Output_Get:double)
   > Set knee_rot_r(knee_rot_r:double=GetAlphaFromReferenceTransformDelta(SkeletalMeshComponent=@n7, BoneName=calf_r, BaseName=thigh_r, DeltaComponentType=QuaternionTwist, InputRange=(X=0,Y=-110), TwistAxis="0,0,1")) -> (Output_Get:double)
   > Set arm_rot_bk_l(arm_rot_bk_l:double=GetAlphaFromReferenceTransformDelta(
     SkeletalMeshComponent=@n7
     BoneName=upperarm_l
     BaseName=clavicle_l
     DeltaComponentType=QuaternionTwist
     InputRange=(X=0,Y=40)
     TwistAxis="1,0,0"
    )) -> (Output_Get:double)
   > Set thigh_rot_up_r(thigh_rot_up_r:double=GetAlphaFromReferenceTransformDelta(SkeletalMeshComponent=@n7, BoneName=thigh_r, BaseName=pelvis, DeltaComponentType=QuaternionTwist, InputRange=(X=0,Y=100), TwistAxis="0,0,1")) -> (Output_Get:double)
   > Set thigh_rot_up_l(thigh_rot_up_l:double=GetAlphaFromReferenceTransformDelta(SkeletalMeshComponent=@n7, BoneName=thigh_l, BaseName=pelvis, DeltaComponentType=QuaternionTwist, InputRange=(X=0,Y=100), TwistAxis="0,0,1")) -> (Output_Get:double)
   > Set arm_rot_up_l(arm_rot_up_l:double=GetAlphaFromReferenceTransformDelta(
     SkeletalMeshComponent=@n7
     BoneName=upperarm_l
     BaseName=clavicle_l
     DeltaComponentType=QuaternionTwist
     InputRange=(X=-55,Y=-110)
     TwistAxis="1,0,0"
    )) -> (Output_Get:double)
   n8 = Get SkelMesh()
   > Set arm_rot_up_r(arm_rot_up_r:double=GetAlphaFromReferenceTransformDelta(
     SkeletalMeshComponent=@n8
     BoneName=upperarm_r
     BaseName=clavicle_r
     DeltaComponentType=QuaternionTwist
     InputRange=(X=0,Y=-120)
     TwistAxis="1,0,0"
    )) -> (Output_Get:double)
   > Set arm_rot_dn_l(arm_rot_dn_l:double=GetAlphaFromReferenceTransformDelta(
     SkeletalMeshComponent=@n8
     BoneName=upperarm_l
     BaseName=spine_05
     DeltaComponentType=QuaternionTwist
     InputRange=(X=0,Y=-30)
     TwistAxis="0,1,0"
    )) -> (Output_Get:double)
   > Set arm_twist_l(arm_twist_l:double=GetAlphaFromReferenceTransformDelta(
     SkeletalMeshComponent=@n8
     BoneName=spine_04
     BaseName=upperarm_twist_01_l
     DeltaComponentType=QuaternionTwist
     InputRange=(X=20,Y=50)
     TwistAxis="1,1,0"
    )) -> (Output_Get:double)
   > Set arm_rot_dn_r(arm_rot_dn_r:double=GetAlphaFromReferenceTransformDelta(
     SkeletalMeshComponent=@n8
     BoneName=upperarm_r
     BaseName=clavicle_r
     DeltaComponentType=QuaternionTwist
     InputRange=(X=0,Y=20)
     TwistAxis="1,0,0"
    )) -> (Output_Get:double)
   n9 = Get Anim_Inst() -> (Anim_Inst:Anim Instance Object Reference)
   Is Valid(exec=@n5.then_2, InputObject=@n9)
   > n10 = Branch(Condition=Get bIsInFrontEnd())
   n11 = Montage_IsPlaying(self=@n9, Montage=/BRCosmetics/Animation/Game/MainPlayer/Menu/BR/CustomSkins/_s21/WayfareMaskF/WayfareMask_Female_Idle_Montage)
   > Set isIdleEmotePlaying(isIdleEmotePlaying:bool=@n11) -> (Output_Get:bool)
   > Set capeConstraintAplha(capeConstraintAplha:double="To Float (Boolean)"(InBool=NOT Boolean(A=@n11))) -> (Output_Get:double)
   Set isIdleEmotePlaying(execute=@n10.else, isIdleEmotePlaying=false)
   > Set capeConstraintAplha(capeConstraintAplha=1)
  ]}
]
variables [
 Anim_Inst {type=AnimInstance, default=null}
 arm_rot_bk_l {type=number}
 arm_rot_dn_l {type=number}
 arm_rot_dn_r {type=number}
 arm_rot_up_l {type=number}
 arm_rot_up_r {type=number}
 arm_twist_l {type=number}
 capeConstraintAplha {type=number}
 Gravity Override Params {
  type=GravityOverrideParamsStruct
  default {
   dBNO_Carried {gravityOverride=<-980 980 -500>}
   onGround_Standing {windAmplitude=<50 -50 10>, windFrequency=<0.15 0.8 0.5>, windOffset=<0 -80 -15>}
  }
 }
 GravityOverride {type=Vector}
 isIdleEmotePlaying {type=boolean}
 isWearBackbling {type=boolean}
 knee_rot_l {type=number}
 knee_rot_r {type=number}
 Player_Pawn {type=FortPlayerPawn}
 SkelMesh {type=SkeletalMeshComponent}
 thigh_rot_up_l {type=number}
 thigh_rot_up_r {type=number}
]
