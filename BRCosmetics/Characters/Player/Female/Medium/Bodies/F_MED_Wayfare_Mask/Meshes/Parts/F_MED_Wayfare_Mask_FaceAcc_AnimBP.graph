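# Animation Blueprint for the face accessory mesh of the female Wayfare Mask character cosmetic.
# It copies the attached parent's pose, then layers a pose copied from BodyMesh over the upper arm and upper arm twist branches.
# A low-alpha Copy Bone then partially pulls head_offset toward the head bone's transform.
# BodyMesh is set to the owner skeletal mesh component when customization loads on an Athena player pawn.
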AnimBlueprint /BRCosmetics/Characters/Player/Female/Medium/Bodies/F_MED_Wayfare_Mask/Meshes/Parts/F_MED_Wayfare_Mask_FaceAcc_AnimBP
target_skeleton=&/F_MED_Wayfare_Mask_FaceAcc_Skeleton
graphs [
 AnimGraph {nodes [
   n1 = Layered blend per bone(
    BasePose=Copy Pose From Mesh() {bUseAttachedParent=true}
    BlendPoses_0=Copy Pose From Mesh(SourceMeshComponent=Get BodyMesh() -> (BodyMesh:Skeletal Mesh Component Object Reference))
   ) {layerSetup [{branchFilters [{boneName=upperarm_twist_01_l}, {boneName=upperarm_twist_01_r}, {boneName=upperarm_l}, {boneName=upperarm_r}]}]}
   n2 = Local To Component(LocalPose=@n1)
   Output Pose(Result=Component To Local(ComponentPose=Copy Bone - Source Bone: head - Target Bone: head_offset!(
      ComponentPose:Component Space Pose Link Structure=@n2
      Alpha:float=0.2
      bCopyTranslation:bool=true
      bCopyRotation:bool=true
      bCopyScale:bool=true
     ) -> (Pose:Component Space Pose Link Structure) {is=title AnimGraphNode_CopyBone, alphaBoolBlend {blendOption=Linear}, sourceBone {boneName=head}, targetBone {boneName=head_offset}})) {layerGroup=None}
  ]}
 EventGraph {nodes [
   Event OnCharacterCustomizationLoaded()
   > Cast To PlayerPawn_Athena(Object=TryGetPawnOwner())
   > Set BodyMesh(BodyMesh=GetOwnerSkeletalMeshComponent())
   Event BlueprintUpdateAnimation()
  ]}
]
variables [BodyMesh {type=SkeletalMeshComponent}]
